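# Translucent unlit, two-sided refraction material for Niagara mesh particles, named as an air-flow effect.
# Swizzlable, vector-driven UV panners take time from a dynamic parameter and accompany a mip-biased noise texture, Fresnel term and powered UV edge ramp for alpha masking.
# Opacity combines tint alpha, particle alpha and an opacity-based depth fade.
# View-space vertex tangents, scaled by the dynamic Ref parameter and temporal-AA dither, form an apparent refraction offset.
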Material /PistolCanary/Assets/VFX/AirFlow/M_AirFlow
graph {
 properties {
  bAutomaticallySetUsageInEditor=false
  bUsedWithNiagaraMeshParticles=true
  blendMode=BLEND_Translucent
  previewMesh=/Game/Effects/Fort_Effects/Mesh/SM_Powerup_Cone_Bottom
  refractionMethod=RM_2DOffset
  shadingModel=MSM_Unlit
  translucencyPass=MTP_BeforeDOF
  twoSided=true
  usedShadingModels=Unlit
 }
 nodes [
  LinearInterpolate(A=TextureCoordinate(), B=TextureCoordinate() {coordinateIndex=1}, Alpha=ScalarParameter() {parameterName=CubeMap_Switch})
  MaterialFunctionCall(Emissive=ParticleColor().RGB) {materialFunction=&MF_FX_EmissiveAdjustments}
  n1 = Transform(None=VertexTangentWS()) {transformSourceType=TRANSFORMSOURCE_World, transformType=TRANSFORM_View}
  n2 = NamedRerouteUsage()
  Multiply(
   A=Multiply(A=Multiply(A=@n1, B=@n2), B=DynamicParameter() {defaultValue=<rgba 1 1 1 0.1>, paramNames [Time,Distort,Dissolve,Ref]}.Ref)
   B=MaterialFunctionCall() {materialFunction=&DitherTemporalAA}.Result
  )
  n3 = TextureCoordinate()
  n4 = MaterialFunctionCall(XY=@n3) {materialFunction=&Swizzle}
  n5 = ScalarParameter() {defaultValue=1, parameterName=UV2_Swizzle}
  n6 = VectorParameter() {parameterName=UV2, defaultValue=<rgba 0.01 0.36 0.25 0.04>}
  NamedRerouteDeclaration(None=MaterialFunctionCall!(
    ClampTexture
    UV=LinearInterpolate(A=@n3, B=@n4.YX, Alpha=@n5)
    Tex
    InRGB=@n6.RGB
    InA=@n6.RGB
    Time=DynamicParameter() {defaultValue=<rgba 1 1 1 0.1>, paramNames [Time,Distort,Dissolve,Ref]}.Time
    TimeMult
   ) {materialFunction=/Game/Effects/Niagara/MasterMaterials/Functions/MF_FX_VectorDrive_TexPanner}.PannerUV) {name=UV_2, nodeColor=<rgba 0 0.4049 1 1>}
  n7 = TextureCoordinate()
  n8 = MaterialFunctionCall(XY=@n7) {materialFunction=&Swizzle}
  n9 = ScalarParameter() {defaultValue=1, parameterName=Main_UV_Swizzle}
  n10 = VectorParameter() {parameterName=Main_UV, defaultValue=<rgba -0.1612 0.9832 -0.18 0>}
  NamedRerouteDeclaration(None=MaterialFunctionCall(
    UV=LinearInterpolate(A=@n7, B=@n8.YX, Alpha=@n9)
    InRGB=@n10.RGB
    InA=@n10.RGB
    Time=DynamicParameter() {defaultValue=<rgba 1 1 1 0.1>, paramNames [Time,Distort,Dissolve,Ref]}.Time
   ) {materialFunction=&MF_FX_VectorDrive_TexPanner}.PannerUV) {name=MainUV, nodeColor=<rgba 0 0.4049 1 1>}
  n11 = OneMinus(None=TextureCoordinate())
  n12 = ScalarParameter() {defaultValue=5, parameterName=AlphaRampPower}
  n13 = Subtract(A=Multiply(A=Power(Base=@n11, Exp=@n12) {constExponent=4}) {constB=2})
  n14 = MaterialFunctionCall(Float2=OneMinus(None=Abs(None=@n13))) {materialFunction=&BreakOutFloat2Components}
  NamedRerouteDeclaration(None=MaterialFunctionCall(A=@n14.R, B=Multiply(A=@n14.R, B=@n14.R), C=@n14.R, Alpha=ScalarParameter() {defaultValue=1, parameterName=AlphaMap_Select}) {materialFunction=&Lerp_3Color}.Result) {name=CubeMap, nodeColor=<rgba 1 1 1 1>}
  n15 = CameraPositionWS()
  n16 = WorldPosition()
  n17 = DotProduct(A=ComponentMask(None=VertexNormalWS()) {b=true, g=true, r=true}, B=Normalize(VectorInput=Subtract(A=@n15, B=@n16.XYZ)))
  n18 = Abs(None=@n17)
  n19 = Power(
   Base=LinearInterpolate(A=OneMinus(None=@n18), B=@n18, Alpha=ScalarParameter() {defaultValue=1, parameterName="Fresnel_Invert (0or1)"})
   Exp=ScalarParameter() {defaultValue=2, parameterName=Fresnel_Power}
  )
  n20 = ScalarParameter() {defaultValue=0.05, parameterName=Fresnel_Offset}
  NamedRerouteDeclaration(None=Multiply(A=Subtract(A=@n19, B=@n20), B=Add(A=@n20))) {name=Fresnel, nodeColor=<rgba 1 0 0.9613 1>}
  n21 = NamedRerouteUsage()
  n22 = TextureSampleParameter2D!(UVs=NamedRerouteUsage(), Bias=ScalarParameter() {defaultValue=5, parameterName=TextureMipBias}, Apply View MipBias) {
   mipValueMode=TMVM_MipBias
   parameterName=NoiseTexture
   samplerType=SAMPLERTYPE_LinearColor
   texture=/Game/Effects/Fort_Effects/Textures/Tile/T_Tile_Directional_T_LightArc
  }
  NamedRerouteDeclaration(None=Multiply(A=Multiply(A=@n21, B=@n22.G), B=NamedRerouteUsage())) {name=Alpha, nodeColor=<rgba 1 0.1875 0 1>}
  n23 = TextureCoordinate()
  n24 = MaterialFunctionCall(XY=@n23) {materialFunction=&Swizzle}
  n25 = ScalarParameter() {defaultValue=1, parameterName=UV1_Swizzle}
  n26 = VectorParameter() {parameterName=UV1, defaultValue=<rgba 0.1033 0.8022 0.332 0>}
  NamedRerouteDeclaration(None=MaterialFunctionCall(
    UV=LinearInterpolate(A=@n23, B=@n24.YX, Alpha=@n25)
    InRGB=@n26.RGB
    InA=@n26.RGB
    Time=DynamicParameter() {defaultValue=<rgba 1 1 1 0.1>, paramNames [Time,Distort,Dissolve,Ref]}.Time
   ) {materialFunction=&MF_FX_VectorDrive_TexPanner}.PannerUV) {name=UV_1, nodeColor=<rgba 0 0.4049 1 1>}
  n27 = MaterialFunctionCall!(
   InAlpha=NamedRerouteUsage()
   OpacityBasedFade_DistanceA=ScalarParameter() {defaultValue=200, group=Opacity, parameterName=OpacityBasedFade_DistanceA}
   OpacityBasedFade_DistanceB=ScalarParameter() {defaultValue=50, group=Opacity, parameterName=OpacityBasedFade_DistanceB}
  ) {materialFunction=/Game/Effects/Niagara/MasterMaterials/Functions/MF_FX_OpacityBasedDepthFade}
  NamedRerouteDeclaration(None=@n27.FadeValue) {name=DepthFadeValue, nodeColor=<rgba 0 1 0.5312 1>}
  n28 = VectorParameter() {parameterName=EmissiveOverAllTint, defaultValue=<rgba 1 1 1 1>}
  n29 = ParticleColor()
  NamedRerouteDeclaration(None=Multiply(A=Multiply(A=@n28.A, B=@n29.A), B=@n27.FinalAlpha)) {name=Opacity, nodeColor=<rgba 0 0 0 1>}
  NamedRerouteUsage()
  Reroute(None=@n17)
 ]
}
