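# M_Shell is an additive, unlit, two-sided Niagara material for a glowing shell in the AirFlow VFX, usable on mesh particles and sprites.
# Arctangent-derived UVs feed a panner driven by the Time dynamic parameter.
# A diagonal wind caustic texture varies a soft sphere-mask radius, alongside invertible Fresnel and depth-fade terms.
# Emissive uses an Emi-driven reciprocal glow curve with boost, exponent and brightness cap, tinted by particle color.
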Material /PistolCanary/Assets/VFX/AirFlow/M_Shell
graph {
 properties {
  bAllowNegativeEmissiveColor=true
  bAutomaticallySetUsageInEditor=false
  bUsedWithNiagaraMeshParticles=true
  bUsedWithNiagaraSprites=true
  blendMode=BLEND_Additive
  shadingModel=MSM_Unlit
  twoSided=true
  usedShadingModels=Unlit
 }
 nodes [
  DepthFade(FadeDistance=ScalarParameter() {defaultValue=100, parameterName=DF_FD})
  n1 = TextureCoordinate()
  LinearInterpolate(
   A=@n1
   B=MaterialFunctionCall(XY=@n1) {materialFunction=&Swizzle}.YX
   Alpha=ScalarParameter() {defaultValue=1, parameterName=UV1_Swizzle}
  )
  n2 = NamedRerouteUsage()
  n3 = DynamicParameter() {paramNames [Time,Emi,Param3,Param4]}
  n4 = Subtract(A=Divide(A=@n3.Emi, B=OneMinus(None=@n2)), B=@n3.Emi)
  n5 = ScalarParameter() {defaultValue=1, parameterName=GlowBoost}
  n6 = Multiply(A=Power(Base=Divide(A=@n4, B=@n5), Exp=ScalarParameter() {defaultValue=1, parameterName=GlowExp}), B=@n5)
  n7 = ScalarParameter() {defaultValue=10, parameterName=MaxiumBright}
  Multiply(A=Multiply(A=Min(A=@n6, B=@n7) {constB=1000}, B=ParticleColor().RGB), B=NamedRerouteUsage())
  n8 = Constant() {r=0.5}
  n9 = TextureSample!(UVs=NamedRerouteUsage(), Tex, Level, Apply View MipBias) {
   mipValueMode=TMVM_MipLevel
   samplerType=SAMPLERTYPE_LinearGrayscale
   texture=/Game/Effects/Fort_Effects/Textures/Tile/T_Caustic_Wind_Diagonal
  }
  Multiply(A=@n8, B=@n9.G)
  NamedRerouteDeclaration(None=SphereMask(A=TextureCoordinate(), B=@n8, Radius=LinearInterpolate(Alpha=@n9.G) {constA=0.25, constB=0.5}) {hardnessPercent=0}) {name=SimpleSphereMask, nodeColor=<rgba 1 1 1 1>}
  n10 = CameraPositionWS()
  n11 = WorldPosition()
  n12 = DotProduct(A=ComponentMask(None=VertexNormalWS()) {b=true, g=true, r=true}, B=Normalize(VectorInput=Subtract(A=@n10, B=@n11.XYZ)))
  n13 = Abs(None=@n12)
  n14 = Power(
   Base=LinearInterpolate(A=OneMinus(None=@n13), B=@n13, Alpha=ScalarParameter() {defaultValue=1, parameterName="Fresnel_Invert (0or1)"})
   Exp=ScalarParameter() {defaultValue=2, parameterName=Fresnel_Power}
  )
  n15 = ScalarParameter() {parameterName=Fresnel_Offset}
  NamedRerouteDeclaration(None=Multiply(A=Subtract(A=@n14, B=@n15), B=Add(A=@n15))) {name=Fresnel, nodeColor=<rgba 1 0 0.9613 1>}
  n16 = MaterialFunctionCall(Float2=Subtract(A=TextureCoordinate()) {constB=0.5}) {materialFunction=&BreakOutFloat2Components}
  n17 = Arctangent2Fast(Y=@n16.R, X=@n16.R)
  n18 = MaterialFunctionCall(Multiplier=Constant() {r=2}) {materialFunction=&Pi}
  n19 = VectorParameter() {parameterName=UV1, defaultValue=<rgba 1 1 -0.3067 0.544>}
  NamedRerouteDeclaration(None=MaterialFunctionCall!(
    ClampTexture
    UV=Divide(A=@n17, B=@n18.Result)
    Tex
    InRGB=@n19.RGB
    InA=@n19.RGB
    Time=DynamicParameter() {paramNames [Time,Emi,Param3,Param4]}.Time
    TimeMult
   ) {materialFunction=/Game/Effects/Niagara/MasterMaterials/Functions/MF_FX_VectorDrive_TexPanner}.PannerUV) {name=UV_1, nodeColor=<rgba 0 0.4049 1 1>}
  Reroute(None=@n12)
 ]
}
